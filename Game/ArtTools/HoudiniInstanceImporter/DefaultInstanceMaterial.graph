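# Default lit material for instanced static meshes, stored with the Houdini instance importer art tools.
# It computes a faceted world-space normal from the normalized cross product of the screen-space derivatives of world position, with tangent-space normals turned off.
# A single scalar parameter accompanies this flat-shading math, and the graph samples no textures.
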Material /Game/ArtTools/HoudiniInstanceImporter/DefaultInstanceMaterial
graph {
 properties {bTangentSpaceNormal=false, bUsedWithInstancedStaticMeshes=true, pixelDepthOffsetMode=PDOM_Legacy, usedShadingModels=Default Lit}
 nodes [
  n1 = WorldPosition()
  Normalize(VectorInput=CrossProduct(A=DDX(Value=@n1.XYZ), B=DDY(Value=@n1.XYZ)))
  ScalarParameter() {defaultValue=0.7}
 ]
}
