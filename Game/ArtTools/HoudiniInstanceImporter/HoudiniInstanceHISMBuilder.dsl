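; Editor tool actor that builds a Hierarchical Instanced Static Mesh from a data table of Houdini-exported points.
; The construction script runs ImportInstances when auto-import is enabled, clearing existing instances before reading the table rows.
; An index modulo keeps every Nth row, swapping Y and Z for Unreal axes.
; Scale comes from pscale or a random range, with optional random yaw, and each transform is also appended to InstanceTransforms.
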(event Custom|ImportInstances
  (bind _randomrotationz (Variables|Default|GetRandomRotationZ))
  (bind _randomrotzmax (Variables|Default|GetRandomRotZMax))
  (bind _randomrotzmin (Variables|Default|GetRandomRotZMin))
  (bind _returnvalue (Math|Random|RandomFloatinRange _randomrotzmin _randomrotzmax))
  (bind _returnvalue_2 (Math|Rotator|MakeRotator 0.0 0.0 _returnvalue))
  (bind _returnvalue_3 (Math|Rotator|SelectRotator _returnvalue_2 "0, 0, 0" _randomrotationz))
  (bind _randomscalemax (Variables|Default|GetRandomScaleMax))
  (bind _randomscalemin (Variables|Default|GetRandomScaleMin))
  (bind _returnvalue_4 (Math|Random|RandomFloatinRange _randomscalemin _randomscalemax))
  (bind _returnvalue_5 (Math|Vector|MakeVector _returnvalue_4 _returnvalue_4 _returnvalue_4))
  (bind _randomscale (Variables|Default|GetRandomScale))
  (CallFunction|ClearInstances)
  (bind _outrownames (DataTable|GetDataTableRowNames (Variables|Default|GetInstanceDataTable)))
  (for (_array_element _array_index) _outrownames
    (if (== (Math|Integer|%25%28Integer%29 _array_index (Math|Float|Truncate (/ 1.0 (Variables|Default|GetInstancesCountToImport)))) 0)
      (bind _returnvalue_6 (Utilities|GetDataTableRow (Variables|Default|GetInstanceDataTable) _array_element))
      (Utilities|Array|Add (Variables|Default|GetInstanceTransforms) (Math|Transform|MakeTransform (Math|Vector|MakeVector (Utilities|Struct|BreakHoudiniInstanceDataLayout _returnvalue_6) (.Pz_8_488530224541EBF7D43F2F99AE2CCEAF (Utilities|Struct|BreakHoudiniInstanceDataLayout _returnvalue_6)) (.Py_7_462CEB644FCF8818190E0DA7A4233236 (Utilities|Struct|BreakHoudiniInstanceDataLayout _returnvalue_6))) _returnvalue_3 (Math|Vector|SelectVector _returnvalue_5 (Math|Vector|MakeVector (.pscale_24_C1C793994E739F1A41356ABC35528C18 (Utilities|Struct|BreakHoudiniInstanceDataLayout _returnvalue_6)) (.pscale_24_C1C793994E739F1A41356ABC35528C18 (Utilities|Struct|BreakHoudiniInstanceDataLayout _returnvalue_6)) (.pscale_24_C1C793994E739F1A41356ABC35528C18 (Utilities|Struct|BreakHoudiniInstanceDataLayout _returnvalue_6))) _randomscale)))
      (Components|InstancedStaticMesh|AddInstance (Variables|Default|GetHierarchicalInstancedStaticMesh) (Math|Transform|MakeTransform (Math|Vector|MakeVector (Utilities|Struct|BreakHoudiniInstanceDataLayout _returnvalue_6) (.Pz_8_488530224541EBF7D43F2F99AE2CCEAF (Utilities|Struct|BreakHoudiniInstanceDataLayout _returnvalue_6)) (.Py_7_462CEB644FCF8818190E0DA7A4233236 (Utilities|Struct|BreakHoudiniInstanceDataLayout _returnvalue_6))) _returnvalue_3 (Math|Vector|SelectVector _returnvalue_5 (Math|Vector|MakeVector (.pscale_24_C1C793994E739F1A41356ABC35528C18 (Utilities|Struct|BreakHoudiniInstanceDataLayout _returnvalue_6)) (.pscale_24_C1C793994E739F1A41356ABC35528C18 (Utilities|Struct|BreakHoudiniInstanceDataLayout _returnvalue_6)) (.pscale_24_C1C793994E739F1A41356ABC35528C18 (Utilities|Struct|BreakHoudiniInstanceDataLayout _returnvalue_6))) _randomscale)) false))))

(event EventBeginPlay)

(event Collision|EventActorBeginOverlap
  (params (OtherActor :type Actor%20Object%20Reference)))

(event EventTick
  (params (DeltaSeconds :type Float%20%28single-precision%29)))

(event Custom|ClearInstances
  (Components|InstancedStaticMesh|ClearInstances (Variables|Default|GetHierarchicalInstancedStaticMesh)))

(fn ConstructionScript
  (if (Variables|Default|GetAutoImportInstances)
    (CallFunction|ImportInstances)
    (Development|PrintString "Imported HISM Instances")))
